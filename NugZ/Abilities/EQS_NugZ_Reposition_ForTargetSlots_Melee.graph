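# EQS query that picks reposition points around the goal for NugZ melee AI, going by its name for target slots.
# Generates a two-ring arc on the querier's side, with ability-driven radii.
# Tests reject unreachable or crowded points, including other AI's destinations, and favour close, forward, dry spots level with the goal.
# A near-identical second option serves as fallback, omitting the leading Reposition-tagged pathfinding filter.
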EnvQuery /NugZ/Abilities/EQS_NugZ_Reposition_ForTargetSlots_Melee
options [
 {
  generator {
   name=EnvQueryGenerator_Donut_0
   class=EnvQueryGenerator_Donut
   properties {
    arcAngle=90
    bDefineArc=true
    bUseSpiralPattern=false
    center=&FortQueryContext_Goal
    numberOfRings=2
    pointsPerRing=13
    arcDirection {lineFrom=&FortQueryContext_Goal, lineTo=&EnvQueryContext_Querier}
    innerRadius {
     dataBinding=&self:EnvQueryGenerator_Donut_0.FortAIDataProvider_Ability_0
     dataField=BehaviorDistance
     defaultValue=600
    }
    outerRadius {
     dataBinding=&self:EnvQueryGenerator_Donut_0.FortAIDataProvider_Ability_1
     dataField=BehaviorDistance
     defaultValue=1200
    }
    projectionData {
     extentX=10
     navigationFilter=&NavFilter_NPC_AvoidAllWater_C
     postProjectionVerticalOffset=30
    }
   }
  }
  tests [
   FortQueryTest_PathfindingBatch_0 {
    class=FortQueryTest_PathfindingBatch
    properties {
     bDefineReferenceValue=false
     boolValue=true
     clampMaxType=None
     clampMinType=None
     context=&EnvQueryContext_Querier
     filterClass=&NavFilter_NPC_AvoidAllWater_C
     filterType=Match
     floatValueMax=1000
     floatValueMin=1000
     multipleContextFilterOp=AllPass
     multipleContextScoreOp=AverageScore
     navDataOverrideContext=null
     normalizationType=Absolute
     pathFromContext=true
     referenceValue=0
     scanRangeMultiplier=1.5
     scoreClampMax=0
     scoreClampMin=0
     scoringEquation=Constant
     scoringFactor=1
     skipUnreachable=true
     testComment=Throw out things you cannot path to.
     testMode=PathExist
     testPurpose=Filter
     navFilterTag {tagName=NavFilter.NPC.Reposition}
    }
   }
   EnvQueryTest_PathfindingBatch_4 {
    class=EnvQueryTest_PathfindingBatch
    properties {filterClass=&NavFilter_NPC_AvoidAllWater_C}
   }
   EnvQueryTest_PathfindingBatch_5 {
    class=EnvQueryTest_PathfindingBatch
    properties {
     context=&FortQueryContext_Goal
     filterClass=&RecastFilter_UseDefaultArea
     testComment=Disabled for perf - Path should also exist to Goal
    }
   }
   EnvQueryTest_Distance_10 {
    class=EnvQueryTest_Distance
    properties {
     distanceTo=&EnvQueryContext_Querier
     filterType=Minimum
     floatValueMin=250
     scoringEquation=Linear
     scoringFactor=1
     testComment=Throw out points we're basically on top of
     testMode=Distance3D
     testPurpose=Filter
    }
   }
   EnvQueryTest_Distance_11 {
    class=EnvQueryTest_Distance
    properties {
     distanceTo=&FortQueryContext_NearbyAIPawns
     filterType=Minimum
     floatValueMin=250
     scoringEquation=Linear
     scoringFactor=1
     testComment=Throw out points near other AI
     testMode=Distance3D
     testPurpose=Filter
    }
   }
   EnvQueryTest_Distance_12 {
    class=EnvQueryTest_Distance
    properties {
     distanceTo=&FortQueryContext_NearbyAIPawnsMoveDestinations
     filterType=Minimum
     floatValueMin=250
     scoringEquation=Linear
     scoringFactor=1
     testComment=Throw out points near other AI's Movement Destinations
     testMode=Distance3D
     testPurpose=Filter
    }
   }
   EnvQueryTest_Distance_13 {
    class=EnvQueryTest_Distance
    properties {
     clampMaxType=SpecifiedValue
     clampMinType=SpecifiedValue
     distanceTo=&EnvQueryContext_Querier
     filterType=Range
     floatValueMin=0
     scoreClampMax=1500
     scoreClampMin=1000
     scoringEquation=InverseLinear
     scoringFactor=0.5
     testComment=Prefer closer points
     testMode=Distance3D
     testPurpose=Score
    }
   }
   EnvQueryTest_Dot_4 {
    class=EnvQueryTest_Dot
    properties {
     clampMaxType=SpecifiedValue
     clampMinType=SpecifiedValue
     scoreClampMax=0.85
     scoreClampMin=-1
     scoringFactor=0.5
     testComment=Prefer things in front of self
    }
   }
   EnvQueryTest_Dot_5 {
    class=EnvQueryTest_Dot
    properties {
     clampMaxType=None
     clampMinType=None
     filterType=Minimum
     floatValueMax=0.707
     floatValueMin=-0.2
     scoreClampMax=707
     scoreClampMin=-707
     scoringEquation=InverseLinear
     scoringFactor=0.2
     testComment=Disabled because we only generate in an arc now - Throw out items on the other side of the goal.
     testPurpose=Filter
     lineA {dirMode=TwoPoints, lineFrom=&FortQueryContext_Goal, lineTo=&EnvQueryContext_Querier}
     lineB {lineFrom=&FortQueryContext_Goal}
    }
   }
   EnvQueryTest_Project_2 {
    class=EnvQueryTest_Project
    properties {
     scoringFactor=2
     testComment=Give LOTS of score to things NOT on Water. We only use Water if we really need to.
     testPurpose=Score
     projectionData {
      navigationFilter=&NavFilter_NPC_ExcludeAllWater_C
      postProjectionVerticalOffset=30
      projectDown=1024
      projectUp=1024
      traceMode=Navigation
     }
    }
   }
   EnvQueryTest_Distance_14 {
    class=EnvQueryTest_Distance
    properties {
     clampMaxType=SpecifiedValue
     clampMinType=SpecifiedValue
     distanceTo=&FortQueryContext_Goal
     filterType=Range
     floatValueMax=300
     floatValueMin=0
     scoreClampMax=150
     scoreClampMin=149
     scoringEquation=InverseLinear
     scoringFactor=2
     testComment=Greatly prefer things on the same Z plane as Goal. Helps with weird water projection/depth issues.
     testMode=DistanceZ
     testPurpose=Score
    }
   }
   EnvQueryTest_Project_4 {
    class=EnvQueryTest_Project
    properties {
     scoringFactor=0
     testComment=Project to any traversable so we don't penetrate under water.
     testPurpose=Score
     projectionData {
      navigationFilter=&NavFilter_NPC_AnyTraversable_C
      postProjectionVerticalOffset=30
      projectDown=1024
      projectUp=1024
      traceMode=Navigation
     }
    }
   }
  ]
 }
 {
  generator {
   name=EnvQueryGenerator_Donut_3
   class=EnvQueryGenerator_Donut
   properties {
    arcAngle=90
    bDefineArc=true
    bUseSpiralPattern=false
    center=&FortQueryContext_Goal
    numberOfRings=2
    pointsPerRing=13
    arcDirection {lineFrom=&FortQueryContext_Goal, lineTo=&EnvQueryContext_Querier}
    innerRadius {
     dataBinding=&self:EnvQueryGenerator_Donut_3.FortAIDataProvider_Ability_0
     dataField=BehaviorDistance
     defaultValue=600
    }
    outerRadius {
     dataBinding=&self:EnvQueryGenerator_Donut_3.FortAIDataProvider_Ability_1
     dataField=BehaviorDistance
     defaultValue=1200
    }
    projectionData {
     extentX=10
     navigationFilter=&NavFilter_NPC_AvoidAllWater_C
     postProjectionVerticalOffset=30
    }
   }
  }
  tests [
   EnvQueryTest_PathfindingBatch_2 {
    class=EnvQueryTest_PathfindingBatch
    properties {filterClass=&NavFilter_NPC_AvoidAllWater_C}
   }
   EnvQueryTest_PathfindingBatch_3 {
    class=EnvQueryTest_PathfindingBatch
    properties {
     context=&FortQueryContext_Goal
     filterClass=&RecastFilter_UseDefaultArea
     testComment=Disabled for perf - Path should also exist to Goal
    }
   }
   EnvQueryTest_Distance_5 {
    class=EnvQueryTest_Distance
    properties {
     distanceTo=&EnvQueryContext_Querier
     filterType=Minimum
     floatValueMin=250
     scoringEquation=Linear
     scoringFactor=1
     testComment=Throw out points we're basically on top of
     testMode=Distance3D
     testPurpose=Filter
    }
   }
   EnvQueryTest_Distance_6 {
    class=EnvQueryTest_Distance
    properties {
     distanceTo=&FortQueryContext_NearbyAIPawns
     filterType=Minimum
     floatValueMin=250
     scoringEquation=Linear
     scoringFactor=1
     testComment=Throw out points near other AI
     testMode=Distance3D
     testPurpose=Filter
    }
   }
   EnvQueryTest_Distance_7 {
    class=EnvQueryTest_Distance
    properties {
     distanceTo=&FortQueryContext_NearbyAIPawnsMoveDestinations
     filterType=Minimum
     floatValueMin=250
     scoringEquation=Linear
     scoringFactor=1
     testComment=Throw out points near other AI's Movement Destinations
     testMode=Distance3D
     testPurpose=Filter
    }
   }
   EnvQueryTest_Distance_8 {
    class=EnvQueryTest_Distance
    properties {
     clampMaxType=SpecifiedValue
     clampMinType=SpecifiedValue
     distanceTo=&EnvQueryContext_Querier
     filterType=Range
     floatValueMin=0
     scoreClampMax=1500
     scoreClampMin=1000
     scoringEquation=InverseLinear
     scoringFactor=0.5
     testComment=Prefer closer points
     testMode=Distance3D
     testPurpose=Score
    }
   }
   EnvQueryTest_Dot_2 {
    class=EnvQueryTest_Dot
    properties {
     clampMaxType=SpecifiedValue
     clampMinType=SpecifiedValue
     scoreClampMax=0.85
     scoreClampMin=-1
     scoringFactor=0.5
     testComment=Prefer things in front of self
    }
   }
   EnvQueryTest_Dot_3 {
    class=EnvQueryTest_Dot
    properties {
     clampMaxType=None
     clampMinType=None
     filterType=Minimum
     floatValueMax=0.707
     floatValueMin=-0.2
     scoreClampMax=707
     scoreClampMin=-707
     scoringEquation=InverseLinear
     scoringFactor=0.2
     testComment=Disabled because we only generate in an arc now - Throw out items on the other side of the goal.
     testPurpose=Filter
     lineA {dirMode=TwoPoints, lineFrom=&FortQueryContext_Goal, lineTo=&EnvQueryContext_Querier}
     lineB {lineFrom=&FortQueryContext_Goal}
    }
   }
   EnvQueryTest_Project_1 {
    class=EnvQueryTest_Project
    properties {
     scoringFactor=2
     testComment=Give LOTS of score to things NOT on Water. We only use Water if we really need to.
     testPurpose=Score
     projectionData {
      navigationFilter=&NavFilter_NPC_ExcludeAllWater_C
      postProjectionVerticalOffset=30
      projectDown=1024
      projectUp=1024
      traceMode=Navigation
     }
    }
   }
   EnvQueryTest_Distance_9 {
    class=EnvQueryTest_Distance
    properties {
     clampMaxType=SpecifiedValue
     clampMinType=SpecifiedValue
     distanceTo=&FortQueryContext_Goal
     filterType=Range
     floatValueMax=300
     floatValueMin=0
     scoreClampMax=150
     scoreClampMin=149
     scoringEquation=InverseLinear
     scoringFactor=2
     testComment=Greatly prefer things on the same Z plane as Goal. Helps with weird water projection/depth issues.
     testMode=DistanceZ
     testPurpose=Score
    }
   }
   EnvQueryTest_Project_5 {
    class=EnvQueryTest_Project
    properties {
     scoringFactor=0
     testComment=Project to any traversable so we don't penetrate under water.
     testPurpose=Score
     projectionData {
      navigationFilter=&NavFilter_NPC_AnyTraversable_C
      postProjectionVerticalOffset=30
      projectDown=1024
      projectUp=1024
      traceMode=Navigation
     }
    }
   }
  ]
 }
]
